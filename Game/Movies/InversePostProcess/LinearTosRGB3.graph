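# LinearTosRGB3 is a material function that converts a Float3 linear color into sRGB, stored among the movie inverse post-process assets.
# It splits the Linear input into R, G and B, runs each channel through the engine's LinearTosRGB function, and recombines the results into a Float3 sRGB output.
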MaterialFunction /Game/Movies/InversePostProcess/LinearTosRGB3
graph {nodes [
  n1 = MaterialFunctionCall(Float3=FunctionInput() {inputName=Linear}) {materialFunction=&BreakOutFloat3Components}
  FunctionOutput(None=MaterialFunctionCall(
    X=MaterialFunctionCall!(Linear=@n1.R) {materialFunction=/Engine/Functions/Engine_MaterialFunctions01/ImageAdjustment/LinearTosRGB}.sRGB
    Y=MaterialFunctionCall(Linear=@n1.G) {materialFunction=&LinearTosRGB}.sRGB
    Z=MaterialFunctionCall(Linear=@n1.B) {materialFunction=&LinearTosRGB}.sRGB
   ) {materialFunction=&MakeFloat3}.Result) {outputName=sRGB}
 ]}
