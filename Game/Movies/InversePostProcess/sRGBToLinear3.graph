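# Material function that converts a three-component sRGB color to linear space, stored with the movie inverse post-process assets.
# It splits the sRGB input into R, G and B with BreakOutFloat3Components and runs each channel through the engine's scalar sRGBToLinear function.
# MakeFloat3 recombines the three converted channels into the Linear output.
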MaterialFunction /Game/Movies/InversePostProcess/sRGBToLinear3
graph {nodes [
  n1 = MaterialFunctionCall(Float3=FunctionInput() {inputName=sRGB}) {materialFunction=&BreakOutFloat3Components}
  FunctionOutput(None=MaterialFunctionCall(
    X=MaterialFunctionCall!(sRGB=@n1.R) {materialFunction=/Engine/Functions/Engine_MaterialFunctions01/ImageAdjustment/sRGBToLinear}.Linear
    Y=MaterialFunctionCall(sRGB=@n1.G) {materialFunction=&sRGBToLinear}.Linear
    Z=MaterialFunctionCall(sRGB=@n1.B) {materialFunction=&sRGBToLinear}.Linear
   ) {materialFunction=&MakeFloat3}.Result) {outputName=Linear}
 ]}
