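# Material function that inverts the game's tonemap on a movie or texture input, using time-of-day inverse LUTs.
# It re-encodes the sample to sRGB, dims it to avoid clipping, then applies half-texel scaling to index four 33-size volume LUTs.
# Color_Time_of_Day combines the Morning, Day, Evening and Night lookups before conversion back to linear.
# The result is divided by eye adaptation to cancel exposure, except on the mobile shading path.
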MaterialFunction /Game/Movies/InversePostProcess/MF_invertFNtonemap
graph {
 properties {bExposeToLibrary=true}
 nodes [
  n1 = TextureCoordinate()
  n2 = TextureObject()
  n3 = MaterialFunctionCall!(Linear=TextureSample(
    UVs=FunctionInput(Preview=@n1) {bUsePreviewValueAsDefault=true, inputName=UVs, inputType=FunctionInput_Vector2}
    Tex=FunctionInput(Preview=@n2) {bUsePreviewValueAsDefault=true, inputName=Texture, inputType=FunctionInput_Texture2D}
   ).RGB) {desc=Needed if texture is brought in with sRGB linearization, materialFunction=&/LinearTosRGB3}
  n4 = Add(
   A=Multiply(A=Multiply(A=@n3.sRGB) {constB=0.86, desc=Dim content to avoid clipping top values}, B=Constant() {desc=32 / 33, r=0.9697})
   B=Constant() {desc=0.5 / 33, r=0.0152}
  )
  n5 = TextureSampleParameterVolume!(UVs=@n4, Apply View MipBias) {parameterName=Inverse Morning, samplerType=SAMPLERTYPE_LinearColor, texture=&/InvFNpapaya_Morning33_Volume}
  n6 = TextureSampleParameterVolume(UVs=@n4) {parameterName=Inverse Day, samplerType=SAMPLERTYPE_LinearColor, texture=&/InvFNpapaya_Day33_Volume}
  n7 = TextureSampleParameterVolume(UVs=@n4) {parameterName=Inverse Evening, samplerType=SAMPLERTYPE_LinearColor, texture=&/InvFNpapaya_Evening33_Volume}
  n8 = TextureSampleParameterVolume(UVs=@n4) {parameterName=Inverse Night, samplerType=SAMPLERTYPE_LinearColor, texture=&/InvFNpapaya_Night33_Volume}
  n9 = MaterialFunctionCall!(sRGB=MaterialFunctionCall(Color_Morning=@n5.RGB, Color_Day=@n6.RGB, Color_Evening=@n7.RGB, Color_Night=@n8.RGB) {materialFunction=&Color_Time_of_Day}.Result) {materialFunction=&/sRGBToLinear3}
  n10 = Divide(A=@n9.Linear, B=EyeAdaptation().EyeAdaptation)
  FunctionOutput(None=ShadingPathSwitch(Default=@n10, Deferred=@n10, Mobile=@n9.Linear) {desc=Migrated from FeatureLevelSwitch})
 ]
}
